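# Niagara idle smoke effect for the BriskReign backpack cosmetic, with all three emitters sampling its skeletal mesh.
# SpiralSmoke burst-spawns persistent mesh and sprite smoke on sockets or bones, feeding smoothed position into material parameters for lag.
# WhispySmoke and WhispySmokeLocal emit dark flipbook puffs that drift downward, the local variant also trailing against smoothed owner velocity.
# System fade and time-of-day color tint all layers; puff opacity also varies by time of day and front end.
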NiagaraSystem /BRCosmetics/Accessories/FORT_Backpacks/Backpack_M_MED_BriskReign/FX/NS_Backpack_BriskReign_Hit_Idle
dependencies {
 usedDataInterfaces [
  &NiagaraDataInterfaceCurve
  &NiagaraDataInterfaceVector2DCurve
  &NiagaraDataInterfaceSpriteRendererInfo
  &NiagaraDataInterfaceColorCurve
  &NiagaraDataInterfaceSkeletalMesh
 ]
 usedDynamicInputs [
  &MakeFloatFromVector
  &Vector2DFromCurve
  &RandomRangeVector2D
  &ColorFromCurve
 ]
 usedModules [
  &SystemState
  &self:SystemSpawnScript.SystemScriptSource.SystemScriptGraph.NiagaraNodeAssignment_3.SetVariables_#1
  &self:SystemSpawnScript.SystemScriptSource.SystemScriptGraph.NiagaraNodeAssignment_2.SetVariables_#2
  &self:SystemSpawnScript.SystemScriptSource.SystemScriptGraph.NiagaraNodeAssignment_0.SetVariables_#3
 ]
}
emitters {
 SpiralSmoke {
  emitterUpdateScript [
   &EmitterState(
    Inactive Response=::NewEnumerator1
    Life Cycle Mode=::NewEnumerator1
    Loop Behavior=::NewEnumerator1
    Loop Duration Mode=::NewEnumerator1
    UseLoopDelay=false
   )
   &SpawnBurst_Instantaneous(Spawn Count=3, Use Spawn Probability=false, UseLoopCountLimit=false)
   &SpawnBurst_Instantaneous(Spawn Group=1, Use Spawn Probability=false, UseLoopCountLimit=false) {module=SpawnBurst_Instantaneous001}
  ]
  particleSpawnScript [
   /SetVariables(Particles.VisibilityTag=$Emitter.SpawnGroup) {isSetParameters=true}
   &PartitionParticles(Number Of Partitions=3)
   &InitializeParticle(
    Color=<rgba 0.024 0.0293 0.04 3.5>
    Mesh Scale=&MergeHelper_VectorSwitch(0=<1.6 0.2 0.2>, 1=<1.5 0.3 0.3>, 2=<1.5 0.3 0.3>, 3=?, Index=$Particles.PartitionParticles.Partition, Num=::NewEnumerator1)
    Mesh Scale Mode=::NewEnumerator4
    UsePositionOffset=true
    Write Material Random=true
   )
   &CameraOffset(Camera Offset Amount=2)
   &InitialMeshOrientation(Mesh Axis To Orient=<0 0 -1>, Mesh Orientation Mode=::NewEnumerator2, Use Rotation Vector=false)
  ]
  particleUpdateScript [
   &ParticleState(Kill Particles When Lifetime Has Elapsed=false)
   &SkeletalMeshLocation(
    Bone / Socket Index=$Particles.PartitionParticles.Partition
    MaskSpawnGroup=false
    Mesh Sampling Type=::NewEnumerator5
    Skeletal Mesh=$System.Skeletal Mesh Reader
    Socket Sampling Mode=::NewEnumerator5
   )
   &SpriteFacingAndAlignment(
    Sprite Alignment=&RotateVectorByQuaternion(Vector To Rotate=<1 0 0>, Quaternion=$Particles.SkeletalMeshLocation.SampledRotation)
    WriteAlignment=true
    WriteFacing=true
   )
   &ScaleMeshSize(Scale Factor=&MakeVector(
     X=&FloatFromCurve(FloatCurve=DataInterface() {propertyValues {Curve {keys [<at 0 0.2 RCIM_Cubic>,<at 0.6 1 RCIM_Cubic>]}}}, CurveIndex=$Emitter.Age)
     Y=1
     Z=1
    ))
   &ScaleColor(
    Scale Alpha=&Multiply_Float(A=0.57, B=$System.SystemFade)
    Scale RGB=&Lerp_Vector(
     A=&MakeVectorFromLinearColorRGB(Color=&RandomRangeLinearColor(Minimum=<rgba 1 1 1 1>, Maximum=<rgba 1.5 1.5 1.5 1>))
     B=&MakeVectorFromLinearColorRGB(Color=$System.TODColorMultiplier)
     Alpha=0.5
    )
    ScaleA=true
    ScaleRGB=true
   )
   &InterpolateOverTime(Rate Of Change=5, Target Vector Value:NiagaraPosition=$Particles.Position, Variable Type=::NewEnumerator2)
   &DynamicMaterialParameters(
    Index 0 Param 1=&RandomRangeFloat(Minimum=1, Maximum=2)
    Index 0 Param 2=5
    Index 0 Param 3=3
    Index 0 Param 4=&RandomRangeFloat(Minimum=1, Maximum=1.6)
    Index 1 Param 4=1
    Index 1 Vector 1:Vector3f=$Particles.InterpolateOverTime.InterpolateOverTime_V3_.Moving Average
    Index 2 Param 1=1
    Index 2 Param 2=1
    Index 2 Param 3=1
    Index 2 Param 4=1
    Module.Index0Float0Write=true
    Module.Index0Float1Write=true
    Module.Index0Float2Write=true
    Module.Index0Float3Write=true
    Module.Index1Float3Write=true
    Module.Index1VectorWrite=true
    Module.Index2Float0Write=true
    Module.Index2Float1Write=false
    Module.Index2Float2Write=false
    Module.Index2Float3Write=false
    Write Parameter Index 1=::NewEnumerator3
    Write Parameter Index 2=::NewEnumerator1
   )
   /SetVariables(
    Particles.DynamicParametersAlt:Vector4f=&MakeVector4(X=1, Y=2, Z=1, W=1)
    Particles.DynamicParametersAlt2:Vector4f=&MakeVector4(X=0, Y=0, Z=0, W=1)
   ) {isSetParameters=true}
   &FN_Sprite_Attributes(
    Color=&Multiply_LinearColorByFloat(Linear Color=<rgba 0 0 0 0.8>, Float=&Multiply_Float(A=$System.SystemFade, B=0.5))
    Create Color Attribute=true
    Create Position Attribute=true
    Create SpriteSize Attribute=true
    Position:NiagaraPosition=&AddVectorToPosition(
     Position=$Particles.Position
     Vector=&RotateVectorByQuaternion(Vector To Rotate=<0 0 0>, Quaternion=$Particles.MeshOrientation)
    )
    SpriteSize=&Lerp_Vector2(A=<10 20>, B=<20 40>, Alpha=$System.SystemFade)
    Write to Module Namespace=true
   )
  ]
  rendererClasses [&NiagaraMeshRendererProperties, &NiagaraSpriteRendererProperties]
  renderers [
   {rendererClass=&NiagaraMeshRendererProperties, rendererIndex=0}
   {rendererClass=&NiagaraSpriteRendererProperties, rendererIndex=1}
  ]
 }
 WhispySmoke {
  emitterUpdateScript [
   &EmitterState(Inactive Response=::NewEnumerator1, Life Cycle Mode=::NewEnumerator1, Loop Duration=2, UseLoopDelay=false)
   &SpawnRate(SpawnRate=7, Use Spawn Probability=false)
  ]
  particleSpawnScript [
   &InitializeParticle(
    Color=<rgba 0.0117 0.0181 0.03 1>
    Lifetime Max=1
    Lifetime Min=0.6
    Lifetime Mode=::NewEnumerator1
    Sprite Rotation Mode=::NewEnumerator2
    Sprite Size Max=<16 16>
    Sprite Size Min=<11 11>
    Sprite Size Mode=::NewEnumerator2
    UsePositionOffset=true
    Write Material Random=true
   )
   &CameraOffset(Camera Offset Amount=10)
   /SetVariables(Particles.CustomRandom:NiagaraInt32=&RandomRangeInt(Minimum=0, Maximum=10)) {isSetParameters=true, module=SetVariables_#4}
   &AddVelocity(
    Rotation Coordinate Space=::World
    Velocity=&RandomRangeVector(Minimum=<0 0 -15>, Maximum=<0 0 -20>)
    Velocity Speed Scale=&RandomRangeFloat(Minimum=1, Maximum=3)
   )
  ]
  particleUpdateScript [
   &ParticleState()
   &SkeletalMeshLocation(
    MaskSpawnGroup=false
    Mesh Sampling Type=::NewEnumerator5
    Position Sampling=::NewEnumerator1
    Sampled Position Offset=&RandomRangeVector(Minimum=<0 0 0>, Maximum=<0 0 10>)
    Skeletal Mesh=$System.Skeletal Mesh Reader
    Socket Sampling Mode=::NewEnumerator5
   )
   &SpriteRotationRate(Rotation Rate=&RandomRangeFloat(Minimum=-50, Maximum=50))
   &ScaleSpriteSize(
    Scale Sprite Size Mode=::NewEnumerator0
    Uniform Scale Factor=&FloatFromCurve(FloatCurve=DataInterface() {propertyValues {Curve {
        defaultValue=1
        keys [
         {arriveTangent=4.1785, interpMode=RCIM_Cubic, leaveTangent=4.1785, tangentMode=RCTM_User, time=0, value=1}
         {arriveTangent=0.6516, interpMode=RCIM_Cubic, leaveTangent=0.6516, tangentMode=RCTM_User, time=0.2042, value=1.7}
         {arriveTangent=0.0224, interpMode=RCIM_Cubic, leaveTangent=0.0224, tangentMode=RCTM_User, time=1, value=2}
        ]
       }}})
   )
   &SubUVAnimation(
    End Frame Range Override:NiagaraInt32=&Add_Integer(A=&RandomRangeInt(Minimum=10, Maximum=12), B=$Particles.Initial.CustomRandom)
    "Sprite Renderer "=DataInterface() {propertyValues {SpriteRenderer=&self:WhispySmoke_0.NiagaraSpriteRendererProperties_0}}
    Start Frame Range Override=$Particles.Initial.CustomRandom
    UseEndFrame=true
    UseStartFrame=true
   )
   &ScaleColor(
    Scale Alpha=&FloatFromCurve(
     FloatCurve=DataInterface() {propertyValues {Curve {keys [
         {arriveTangent=3.7485, interpMode=RCIM_Cubic, leaveTangent=3.7485, tangentMode=RCTM_User, time=0, value=0.2}
         {interpMode=RCIM_Cubic, tangentMode=RCTM_User, time=0.5, value=1}
         {tangentMode=RCTM_User, time=1, value=0}
        ]}}}
     Scale Curve=&Multiply_Float(
      A=&FN_FrontEndFloatModifier(Gameplay=&FNTimeOfDayScalar(Night Scale=0.6, Morning Scale=0.9, Evening Scale=0.8), Front End=1)
      B=$System.SystemFade
     )
    )
    Scale RGB=&MakeVectorFromLinearColorRGB(Color=$System.TODColorMultiplier)
    ScaleA=true
    ScaleRGB=true
   )
   &Drag(Drag=&RandomRangeFloat(Minimum=2, Maximum=3), Use Linear Drag=true, Use Rotational Drag=true)
   &SolveForcesAndVelocity(Clamp Velocity=false, Limit Acceleration=false)
  ]
  rendererClasses [&NiagaraSpriteRendererProperties]
  renderers [{rendererClass=&NiagaraSpriteRendererProperties}]
 }
 WhispySmokeLocal {
  emitterUpdateScript [
   &EmitterState(Inactive Response=::NewEnumerator1, Life Cycle Mode=::NewEnumerator1, Loop Duration=2, UseLoopDelay=false)
   &SpawnRate(SpawnRate=6, Use Spawn Probability=false)
  ]
  particleSpawnScript [
   &InitializeParticle(
    Color=<rgba 0.0078 0.0121 0.02 1.3>
    Lifetime=&MakeCustomFloatFromBool(
     Bool=&SetBoolByFloatComparison(A=750, B=&VectorLength(Vector=$Engine.Owner.Velocity))
     True Float=&RandomRangeFloat(Minimum=1.5, Maximum=3)
     False Float=&RandomRangeFloat(Minimum=1, Maximum=1.5)
    )
    Sprite Rotation Mode=::NewEnumerator2
    Sprite Size Max=<15 15>
    Sprite Size Min=<11 11>
    Sprite Size Mode=::NewEnumerator2
    UsePositionOffset=true
    Write Material Random=true
   )
   &CameraOffset(Camera Offset Amount=10)
   /SetVariables(Particles.CustomRandom:NiagaraInt32=&RandomRangeInt(Minimum=0, Maximum=10)) {isSetParameters=true, module=SetVariables_#4}
  ]
  particleUpdateScript [
   &ParticleState()
   &InterpolateOverTime(
    Rate Of Change=5
    Target Vector Value:Vector3f=&Multiply_VectorByFloat(
     Vector=$Engine.Owner.Velocity
     Float=&RemapRange(
      Input Value To Remap=1000
      Input Min=750
      Input Max=1000
      Output Min=0
      Output Max=&MakeCustomFloatFromBool(
       Bool=&SetBoolByFloatComparison(A=750, B=&VectorLength(Vector=$Engine.Owner.Velocity))
       True Float=-0.1
       False Float=-0.05
      )
     )
    )
    Variable Type=::NewEnumerator2
   )
   &SkeletalMeshLocation(
    MaskSpawnGroup=false
    Mesh Sampling Type=::NewEnumerator5
    Sampled Position Offset=&AddVectorToPosition(
     Position=&Lerp_Vector(
      A=<0 0 0>
      B=<0 0 -35>
      Alpha=&FloatFromCurve(FloatCurve=DataInterface() {propertyValues {Curve {defaultValue=1, keys [{tangentMode=RCTM_User, time=0, value=0}, {tangentMode=RCTM_User, time=1, value=1}]}}})
     )
     Vector=&Multiply_VectorByFloat(
      Vector=$Particles.InterpolateOverTime.InterpolateOverTime_V3_.Moving Average
      Float=&FloatFromCurve(FloatCurve=DataInterface() {propertyValues {Curve {
          defaultValue=1
          keys [
           {arriveTangent=-0.1546, interpMode=RCIM_Cubic, leaveTangent=-0.1546, tangentMode=RCTM_User, time=0, value=0}
           {arriveTangent=1.6848, interpMode=RCIM_Cubic, leaveTangent=1.6848, tangentMode=RCTM_User, time=1, value=1}
          ]
         }}})
     )
    )
    Sampled Position Offset Coordinate Space:ENiagaraSkelMeshTransforms=::NewEnumerator0
    Skeletal Mesh=$System.Skeletal Mesh Reader
    Socket Sampling Mode=::NewEnumerator5
   )
   &SpriteRotationRate(Rotation Rate=&RandomRangeFloat(Minimum=-50, Maximum=50))
   &ScaleSpriteSize(
    Scale Sprite Size Mode=::NewEnumerator0
    Uniform Scale Factor=&FloatFromCurve(FloatCurve=DataInterface() {propertyValues {Curve {
        defaultValue=1
        keys [
         {arriveTangent=4.1785, interpMode=RCIM_Cubic, leaveTangent=4.1785, tangentMode=RCTM_User, time=0, value=1}
         {arriveTangent=0.6516, interpMode=RCIM_Cubic, leaveTangent=0.6516, tangentMode=RCTM_User, time=0.2042, value=1.7}
         {arriveTangent=0.0224, interpMode=RCIM_Cubic, leaveTangent=0.0224, tangentMode=RCTM_User, time=1, value=2}
        ]
       }}})
   )
   &SubUVAnimation(
    End Frame Range Override:NiagaraInt32=&Add_Integer(A=&RandomRangeInt(Minimum=15, Maximum=20), B=$Particles.Initial.CustomRandom)
    "Sprite Renderer "=DataInterface() {propertyValues {SpriteRenderer=&self:WhispySmokeLocal_0.NiagaraSpriteRendererProperties_0}}
    Start Frame Range Override=$Particles.Initial.CustomRandom
    UseEndFrame=true
    UseStartFrame=true
   )
   &ScaleColor(
    Scale Alpha=&FloatFromCurve(
     FloatCurve=DataInterface() {propertyValues {Curve {keys [
         {arriveTangent=3.7485, interpMode=RCIM_Cubic, leaveTangent=3.7485, tangentMode=RCTM_User, time=0, value=0.2}
         {interpMode=RCIM_Cubic, tangentMode=RCTM_User, time=0.5, value=1}
         {tangentMode=RCTM_User, time=1, value=0}
        ]}}}
     Scale Curve=&Multiply_Float(
      A=&FN_FrontEndFloatModifier(Gameplay=&FNTimeOfDayScalar(Night Scale=0.6, Morning Scale=0.9, Evening Scale=0.8), Front End=1)
      B=$System.SystemFade
     )
    )
    Scale RGB=&MakeVectorFromLinearColorRGB(Color=$System.TODColorMultiplier)
    ScaleA=true
    ScaleRGB=true
   )
   &SolveForcesAndVelocity(Clamp Velocity=false, Limit Acceleration=false)
  ]
  rendererClasses [&NiagaraSpriteRendererProperties]
  renderers [{rendererClass=&NiagaraSpriteRendererProperties}]
 }
}
summary {
 systemName=NS_Backpack_BriskReign_Hit_Idle
 emitters [
  {emitterName=SpiralSmoke, rendererClasses [&NiagaraMeshRendererProperties, &NiagaraSpriteRendererProperties]}
  {emitterName=WhispySmokeLocal, rendererClasses [&NiagaraSpriteRendererProperties]}
  {emitterName=WhispySmoke, rendererClasses [&NiagaraSpriteRendererProperties]}
 ]
 userVariables [
  User.bIsFrontEnd {type=NiagaraBool}
  User.bIsFrontEndPreview {type=NiagaraBool}
 ]
}
