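# Physics asset for the M_MED_BriskReign backpack cape, named as the crouch variant.
# Left and right dyn_cape chains each link eight identical capsules end to end from a tiny MSTR anchor capsule, with tight, uniform swing and twist limits on every joint.
# A large, thin, unconstrained box on dyn_cape_01 sits apart from both chains, apparently serving as a flat collision plane for the cape.
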PhysicsAsset /BRCosmetics/Accessories/FORT_Backpacks/Backpack_M_MED_BriskReign/Meshes/M_MED_BriskReign_Pack_Physics_Crouch
bodies {
 dyn_cape_01 [{shapeType=Box, extentX=200, extentY=200, extentZ=5, shapeName=dyn_cape_01_box, center=<50 0 0>, rotation=<rot 0 0 90>}]
 dyn_cape_01_l [{shapeType=Capsule, length=7.5, radius=6.005, shapeName=B, center=<6.5042 0 0>, rotation=<rot 90 -179.9999 0>}]
 dyn_cape_01_l_MSTR [{shapeType=Capsule, length=0.01, radius=0.01}]
 dyn_cape_01_r [{shapeType=Capsule, length=7.5, radius=6.005, shapeName=B, center=<6.5042 0 0>, rotation=<rot 90 -179.9999 0>}]
 dyn_cape_01_r_MSTR [{shapeType=Capsule, length=0.01, radius=0.01}]
 dyn_cape_02_l [{shapeType=Capsule, length=7.5, radius=6.005, shapeName=B, center=<6.5042 0 0>, rotation=<rot 90 -179.9999 0>}]
 dyn_cape_02_r [{shapeType=Capsule, length=7.5, radius=6.005, shapeName=B, center=<6.5042 0 0>, rotation=<rot 90 -179.9999 0>}]
 dyn_cape_03_l [{shapeType=Capsule, length=7.5, radius=6.005, shapeName=B, center=<6.5042 0 0>, rotation=<rot 90 -179.9999 0>}]
 dyn_cape_03_r [{shapeType=Capsule, length=7.5, radius=6.005, shapeName=B, center=<6.5042 0 0>, rotation=<rot 90 -179.9999 0>}]
 dyn_cape_04_l [{shapeType=Capsule, length=7.5, radius=6.005, shapeName=B, center=<6.5042 0 0>, rotation=<rot 90 -179.9999 0>}]
 dyn_cape_04_r [{shapeType=Capsule, length=7.5, radius=6.005, shapeName=B, center=<6.5042 0 0>, rotation=<rot 90 -179.9999 0>}]
 dyn_cape_05_l [{shapeType=Capsule, length=7.5, radius=6.005, shapeName=B, center=<6.5042 0 0>, rotation=<rot 90 -179.9999 0>}]
 dyn_cape_05_r [{shapeType=Capsule, length=7.5, radius=6.005, shapeName=B, center=<6.5042 0 0>, rotation=<rot 90 -179.9999 0>}]
 dyn_cape_06_l [{shapeType=Capsule, length=7.5, radius=6.005, shapeName=B, center=<6.5042 0 0>, rotation=<rot 90 -179.9999 0>}]
 dyn_cape_06_r [{shapeType=Capsule, length=7.5, radius=6.005, shapeName=B, center=<6.5042 0 0>, rotation=<rot 90 -179.9999 0>}]
 dyn_cape_07_l [{shapeType=Capsule, length=7.5, radius=6.005, shapeName=B, center=<6.5042 0 0>, rotation=<rot 90 -179.9999 0>}]
 dyn_cape_07_r [{shapeType=Capsule, length=7.5, radius=6.005, shapeName=B, center=<6.5042 0 0>, rotation=<rot 90 -179.9999 0>}]
 dyn_cape_08_l [{shapeType=Capsule, length=7.5, radius=6.005, shapeName=B, center=<6.5042 0 0>, rotation=<rot 90 -179.9999 0>}]
 dyn_cape_08_r [{shapeType=Capsule, length=7.5, radius=6.005, shapeName=B, center=<6.5042 0 0>, rotation=<rot 90 -179.9999 0>}]
}
constraints [
 {bone1Name=dyn_cape_01_l, bone2Name=dyn_cape_01_l_MSTR, swing1LimitDegrees=5, swing2LimitDegrees=5, twistLimitDegrees=5}
 {bone1Name=dyn_cape_02_l, bone2Name=dyn_cape_01_l, swing1LimitDegrees=5, swing2LimitDegrees=5, twistLimitDegrees=5}
 {bone1Name=dyn_cape_03_l, bone2Name=dyn_cape_02_l, swing1LimitDegrees=5, swing2LimitDegrees=5, twistLimitDegrees=5}
 {bone1Name=dyn_cape_04_l, bone2Name=dyn_cape_03_l, swing1LimitDegrees=5, swing2LimitDegrees=5, twistLimitDegrees=5}
 {bone1Name=dyn_cape_05_l, bone2Name=dyn_cape_04_l, swing1LimitDegrees=5, swing2LimitDegrees=5, twistLimitDegrees=5}
 {bone1Name=dyn_cape_06_l, bone2Name=dyn_cape_05_l, swing1LimitDegrees=5, swing2LimitDegrees=5, twistLimitDegrees=5}
 {bone1Name=dyn_cape_07_l, bone2Name=dyn_cape_06_l, swing1LimitDegrees=5, swing2LimitDegrees=5, twistLimitDegrees=5}
 {bone1Name=dyn_cape_08_l, bone2Name=dyn_cape_07_l, swing1LimitDegrees=5, swing2LimitDegrees=5, twistLimitDegrees=5}
 {bone1Name=dyn_cape_01_r, bone2Name=dyn_cape_01_r_MSTR, swing1LimitDegrees=5, swing2LimitDegrees=5, twistLimitDegrees=5}
 {bone1Name=dyn_cape_02_r, bone2Name=dyn_cape_01_r, swing1LimitDegrees=5, swing2LimitDegrees=5, twistLimitDegrees=5}
 {bone1Name=dyn_cape_03_r, bone2Name=dyn_cape_02_r, swing1LimitDegrees=5, swing2LimitDegrees=5, twistLimitDegrees=5}
 {bone1Name=dyn_cape_04_r, bone2Name=dyn_cape_03_r, swing1LimitDegrees=5, swing2LimitDegrees=5, twistLimitDegrees=5}
 {bone1Name=dyn_cape_05_r, bone2Name=dyn_cape_04_r, swing1LimitDegrees=5, swing2LimitDegrees=5, twistLimitDegrees=5}
 {bone1Name=dyn_cape_06_r, bone2Name=dyn_cape_05_r, swing1LimitDegrees=5, swing2LimitDegrees=5, twistLimitDegrees=5}
 {bone1Name=dyn_cape_07_r, bone2Name=dyn_cape_06_r, swing1LimitDegrees=5, swing2LimitDegrees=5, twistLimitDegrees=5}
 {bone1Name=dyn_cape_08_r, bone2Name=dyn_cape_07_r, swing1LimitDegrees=5, swing2LimitDegrees=5, twistLimitDegrees=5}
]
