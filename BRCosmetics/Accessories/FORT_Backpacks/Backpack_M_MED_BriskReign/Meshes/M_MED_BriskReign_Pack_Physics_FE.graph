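# Physics asset for the BriskReign backpack, defining simulation bodies for its left and right cape bone chains.
# Each side has eight capsule bodies, slightly thicker from the fourth bone down, constrained in sequence back to a tiny MSTR root body.
# All joints share the same moderate swing and tight twist limits, letting the cape bend with little twisting.
# No constraints link the left chain to the right.
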PhysicsAsset /BRCosmetics/Accessories/FORT_Backpacks/Backpack_M_MED_BriskReign/Meshes/M_MED_BriskReign_Pack_Physics_FE
bodies {
 dyn_cape_01_l [{shapeType=Capsule, length=9, radius=4.505, shapeName=B, center=<6.5042 0 0>, rotation=<rot 90 0 179.9999>}]
 dyn_cape_01_l_MSTR [{shapeType=Capsule, length=0.01, radius=0.01}]
 dyn_cape_01_r [{shapeType=Capsule, length=9, radius=4.505, shapeName=B, center=<6.5042 0 0>, rotation=<rot 90 0 179.9999>}]
 dyn_cape_01_r_MSTR [{shapeType=Capsule, length=0.01, radius=0.01}]
 dyn_cape_02_l [{shapeType=Capsule, length=9, radius=4.505, shapeName=B, center=<6.5042 0 0>, rotation=<rot 90 0 179.9999>}]
 dyn_cape_02_r [{shapeType=Capsule, length=9, radius=4.505, shapeName=B, center=<6.5042 0 0>, rotation=<rot 90 0 179.9999>}]
 dyn_cape_03_l [{shapeType=Capsule, length=9, radius=4.505, shapeName=B, center=<6.5042 0 0>, rotation=<rot 90 0 179.9999>}]
 dyn_cape_03_r [{shapeType=Capsule, length=9, radius=4.505, shapeName=B, center=<6.5042 0 0>, rotation=<rot 90 0 179.9999>}]
 dyn_cape_04_l [{shapeType=Capsule, length=8.5, radius=5.005, shapeName=B, center=<6.5042 0 0>, rotation=<rot 90 0 179.9999>}]
 dyn_cape_04_r [{shapeType=Capsule, length=8.5, radius=5.005, shapeName=B, center=<6.5042 0 0>, rotation=<rot 90 0 179.9999>}]
 dyn_cape_05_l [{shapeType=Capsule, length=8.5, radius=5.005, shapeName=B, center=<6.5042 0 0>, rotation=<rot 90 0 179.9999>}]
 dyn_cape_05_r [{shapeType=Capsule, length=8.5, radius=5.005, shapeName=B, center=<6.5042 0 0>, rotation=<rot 90 0 179.9999>}]
 dyn_cape_06_l [{shapeType=Capsule, length=8.5, radius=5.005, shapeName=B, center=<6.5042 0 0>, rotation=<rot 90 0 179.9999>}]
 dyn_cape_06_r [{shapeType=Capsule, length=8.5, radius=5.005, shapeName=B, center=<6.5042 0 0>, rotation=<rot 90 0 179.9999>}]
 dyn_cape_07_l [{shapeType=Capsule, length=8.5, radius=5.005, shapeName=B, center=<6.5042 0 0>, rotation=<rot 90 0 179.9999>}]
 dyn_cape_07_r [{shapeType=Capsule, length=8.5, radius=5.005, shapeName=B, center=<6.5042 0 0>, rotation=<rot 90 0 179.9999>}]
 dyn_cape_08_l [{shapeType=Capsule, length=8.5, radius=5.005, shapeName=B, center=<6.5042 0 0>, rotation=<rot 90 0 179.9999>}]
 dyn_cape_08_r [{shapeType=Capsule, length=8.5, radius=5.005, shapeName=B, center=<6.5042 0 0>, rotation=<rot 90 0 179.9999>}]
}
constraints [
 {bone1Name=dyn_cape_01_l, bone2Name=dyn_cape_01_l_MSTR, swing1LimitDegrees=45, swing2LimitDegrees=45, twistLimitDegrees=5}
 {bone1Name=dyn_cape_02_l, bone2Name=dyn_cape_01_l, swing1LimitDegrees=45, swing2LimitDegrees=45, twistLimitDegrees=5}
 {bone1Name=dyn_cape_03_l, bone2Name=dyn_cape_02_l, swing1LimitDegrees=45, swing2LimitDegrees=45, twistLimitDegrees=5}
 {bone1Name=dyn_cape_04_l, bone2Name=dyn_cape_03_l, swing1LimitDegrees=45, swing2LimitDegrees=45, twistLimitDegrees=5}
 {bone1Name=dyn_cape_05_l, bone2Name=dyn_cape_04_l, swing1LimitDegrees=45, swing2LimitDegrees=45, twistLimitDegrees=5}
 {bone1Name=dyn_cape_06_l, bone2Name=dyn_cape_05_l, swing1LimitDegrees=45, swing2LimitDegrees=45, twistLimitDegrees=5}
 {bone1Name=dyn_cape_07_l, bone2Name=dyn_cape_06_l, swing1LimitDegrees=45, swing2LimitDegrees=45, twistLimitDegrees=5}
 {bone1Name=dyn_cape_08_l, bone2Name=dyn_cape_07_l, swing1LimitDegrees=45, swing2LimitDegrees=45, twistLimitDegrees=5}
 {bone1Name=dyn_cape_01_r, bone2Name=dyn_cape_01_r_MSTR, swing1LimitDegrees=45, swing2LimitDegrees=45, twistLimitDegrees=5}
 {bone1Name=dyn_cape_02_r, bone2Name=dyn_cape_01_r, swing1LimitDegrees=45, swing2LimitDegrees=45, twistLimitDegrees=5}
 {bone1Name=dyn_cape_03_r, bone2Name=dyn_cape_02_r, swing1LimitDegrees=45, swing2LimitDegrees=45, twistLimitDegrees=5}
 {bone1Name=dyn_cape_04_r, bone2Name=dyn_cape_03_r, swing1LimitDegrees=45, swing2LimitDegrees=45, twistLimitDegrees=5}
 {bone1Name=dyn_cape_05_r, bone2Name=dyn_cape_04_r, swing1LimitDegrees=45, swing2LimitDegrees=45, twistLimitDegrees=5}
 {bone1Name=dyn_cape_06_r, bone2Name=dyn_cape_05_r, swing1LimitDegrees=45, swing2LimitDegrees=45, twistLimitDegrees=5}
 {bone1Name=dyn_cape_07_r, bone2Name=dyn_cape_06_r, swing1LimitDegrees=45, swing2LimitDegrees=45, twistLimitDegrees=5}
 {bone1Name=dyn_cape_08_r, bone2Name=dyn_cape_07_r, swing1LimitDegrees=45, swing2LimitDegrees=45, twistLimitDegrees=5}
]
