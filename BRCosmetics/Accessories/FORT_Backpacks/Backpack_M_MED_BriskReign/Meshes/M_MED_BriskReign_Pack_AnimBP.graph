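# Animation Blueprint for the BriskReign backpack's cape, layering procedural secondary motion on the input pose.
# Additive bone offsets driven by spine bend and jumping feed Anim Dynamics, Trail controllers, then two blended RigidBody simulations (standard and FE physics assets).
# The event graph computes gravity override, spine-based rotations and gender, front-end and emote alphas.
# Crouch physics and FE rotation nodes sit outside the output chain.
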AnimBlueprint /BRCosmetics/Accessories/FORT_Backpacks/Backpack_M_MED_BriskReign/Meshes/M_MED_BriskReign_Pack_AnimBP
target_skeleton=&/M_MED_BriskReign_Pack_Skeleton
graphs [
 AnimGraph {nodes [
   n1 = RigidBody!(ComponentPose, Alpha, OverrideWorldGravity:Vector=Get GravityOverride(), ExternalForce) -> (Pose) {
    bClampLinearTranslationLimitToRefPose=true
    bOverrideWorldGravity=true
    bUseExternalClothCollision=true
    overridePhysicsAsset=&/M_MED_BriskReign_Pack_Physics_Crouch
    simulationSpace=BaseBoneSpace
    alphaBoolBlend {blendInTime=1, blendOption=Linear, blendOutTime=2}
    baseBoneRef {boneName=backpack_root}
    componentAppliedLinearAccClamp=<5000 5000 5000>
    overrideWorldGravity {z=-980}
    simSpaceSettings {
     dampingAlpha=0.3
     maxAngularAcceleration=40
     maxAngularVelocity=8
     maxLinearAcceleration=3500
     maxLinearVelocity=700
     velocityScaleZ=0.8
     worldAlpha=0.5
     externalLinearDragV=<0.2 0.2 0.2>
    }
   }
   Blend Poses by bool(BlendPose_0=Component To Local(ComponentPose=@n1)) {blendType=Linear}
   Local To Component(LocalPose=Use cached pose 'trailControlCache'())
   n2 = "Transform (Modify) Bone - Bone: dyn_cape_01_MSTR"(Alpha=1, Translation="0,0,0", Rotation=Get Cape Rotate FE() -> (Cape Rotate FE:Rotator), Scale="1,1,1") {
    rotationMode=BMM_Additive
    rotationSpace=BCS_BoneSpace
    translationMode=BMM_Additive
    translationSpace=BCS_BoneSpace
    alphaBoolBlend {blendOption=Linear}
    alphaScaleBiasClamp {bInterpResult=true}
    boneToModify {boneName=dyn_cape_01_MSTR}
   }
   n3 = "Transform (Modify) Bone - Bone: backpack_root"(ComponentPose=@n2, Alpha=1, Translation="0,0,0", Rotation=Get Backpack Rotate FE() -> (Backpack Rotate FE:Rotator), Scale="1,1,1") {
    rotationMode=BMM_Additive
    rotationSpace=BCS_BoneSpace
    translationMode=BMM_Additive
    translationSpace=BCS_BoneSpace
    alphaBoolBlend {blendOption=Linear}
    alphaScaleBiasClamp {bInterpResult=true}
    boneToModify {boneName=backpack_root}
   }
   "Transform (Modify) Bone - Bone: backpack_root"(ComponentPose=@n3, Alpha=1, Translation="0,0,0", Rotation="0,0,15", Scale="1,1,1") {
    rotationMode=BMM_Additive
    rotationSpace=BCS_BoneSpace
    translationMode=BMM_Additive
    translationSpace=BCS_BoneSpace
    alphaBoolBlend {blendOption=Linear}
    alphaScaleBiasClamp {bInterpResult=true}
    boneToModify {boneName=backpack_root}
    rotation {roll=15}
   }
   n4 = RigidBody(ComponentPose=Local To Component(LocalPose=Use cached pose 'trailControlCache'()), OverrideWorldGravity=Get GravityOverride()) {
    bClampLinearTranslationLimitToRefPose=true
    bOverrideWorldGravity=true
    bUseExternalClothCollision=true
    overridePhysicsAsset=&/M_MED_BriskReign_Pack_Physics
    simulationSpace=BaseBoneSpace
    alphaBoolBlend {blendInTime=1, blendOption=Linear, blendOutTime=2}
    baseBoneRef {boneName=backpack_root}
    componentAppliedLinearAccClamp=<0 0 0>
    overrideWorldGravity {z=-980}
    simSpaceSettings {
     dampingAlpha=0
     maxAngularAcceleration=300
     maxAngularVelocity=8
     maxLinearAcceleration=300
     maxLinearVelocity=700
     worldAlpha=1
     externalLinearDragV=<1 1 1>
    }
   }
   n5 = RigidBody(ComponentPose=Local To Component(LocalPose=Use cached pose 'trailControlCache'()), OverrideWorldGravity=Get GravityOverride()) {
    bClampLinearTranslationLimitToRefPose=true
    bOverrideWorldGravity=true
    bUseExternalClothCollision=true
    overridePhysicsAsset=&/M_MED_BriskReign_Pack_Physics_FE
    simulationSpace=BaseBoneSpace
    alphaBoolBlend {blendInTime=1, blendOption=Linear, blendOutTime=2}
    baseBoneRef {boneName=backpack_root}
    componentAppliedLinearAccClamp=<5000 5000 5000>
    overrideWorldGravity {z=-980}
    simSpaceSettings {
     dampingAlpha=0.3
     maxAngularAcceleration=40
     maxAngularVelocity=8
     maxLinearAcceleration=3500
     maxLinearVelocity=700
     velocityScaleZ=0.8
     worldAlpha=0.5
     externalLinearDragV=<0.2 0.2 0.2>
    }
   }
   Output Pose(Result=Blend Poses by bool(BlendPose_0=Component To Local(ComponentPose=@n5), BlendPose_1=Component To Local(ComponentPose=@n4)) {blendType=Linear}) {layerGroup=None}
   n6 = "Transform (Modify) Bone - Bone: backpack_root"(ComponentPose=Local To Component(LocalPose=Input Pose()), Alpha=1, Translation="0,-2,0", Rotation="0,0,0", Scale="1,1,1") {
    rotationMode=BMM_Additive
    rotationSpace=BCS_BoneSpace
    translationMode=BMM_Additive
    translationSpace=BCS_BoneSpace
    alphaBoolBlend {blendOption=Linear}
    alphaScaleBiasClamp {bInterpResult=true}
    boneToModify {boneName=backpack_root}
    translation {y=-2}
   }
   n7 = "Transform (Modify) Bone - Bone: backpack_root"(ComponentPose=@n6, Alpha=1, Translation="0,0,0", Rotation="0,0,0", Scale="1,1,1") {
    rotationMode=BMM_Additive
    rotationSpace=BCS_BoneSpace
    translationMode=BMM_Additive
    translationSpace=BCS_BoneSpace
    alphaBoolBlend {blendOption=Linear}
    alphaScaleBiasClamp {bInterpResult=true}
    boneToModify {boneName=backpack_root}
   }
   n8 = "Transform (Modify) Bone - Bone: dyn_cape_01_MSTR"(
    ComponentPose=@n7
    Alpha=1
    Translation=Get Backpack Translation() -> (Backpack Translation:Vector)
    Rotation=Get Backpack Rotate() -> (Backpack Rotate:Rotator)
    Scale="1,1,1"
   ) {
    rotationMode=BMM_Additive
    rotationSpace=BCS_BoneSpace
    translationMode=BMM_Additive
    translationSpace=BCS_BoneSpace
    alphaBoolBlend {blendOption=Linear}
    boneToModify {boneName=dyn_cape_01_MSTR}
   }
   Save cached pose 'initCache'(Pose=Component To Local(ComponentPose=@n8)) {cacheName=initCache}
   n9 = Get JumpingAlpha() -> (JumpingAlpha:float)
   n10 = "Transform (Modify) Bone - Bone: dyn_cape_01_l_MSTR"(ComponentPose=Local To Component(LocalPose=Use cached pose 'initCache'()), Alpha=@n9, Translation="0,0,0", Rotation="40,0,0", Scale="1,1,1") {
    rotationMode=BMM_Additive
    rotationSpace=BCS_BoneSpace
    translationMode=BMM_Additive
    translationSpace=BCS_BoneSpace
    alphaBoolBlend {blendOption=Linear}
    alphaScaleBiasClamp {bInterpResult=true}
    boneToModify {boneName=dyn_cape_01_l_MSTR}
    rotation {pitch=40}
   }
   n11 = "Transform (Modify) Bone - Bone: dyn_cape_01_r_MSTR"(ComponentPose=@n10, Alpha=@n9, Translation="0,0,0", Rotation="40,0,0", Scale="1,1,1") {
    rotationMode=BMM_Additive
    rotationSpace=BCS_BoneSpace
    translationMode=BMM_Additive
    translationSpace=BCS_BoneSpace
    alphaBoolBlend {blendOption=Linear}
    alphaScaleBiasClamp {bInterpResult=true}
    boneToModify {boneName=dyn_cape_01_r_MSTR}
    rotation {pitch=40}
   }
   n12 = "Transform (Modify) Bone - Bone: dyn_cape_01_MSTR"(ComponentPose=@n11, Alpha=@n9, Translation="0,0,0", Rotation="10,0,0", Scale="1,1,1") {
    rotationMode=BMM_Additive
    rotationSpace=BCS_BoneSpace
    translationMode=BMM_Additive
    translationSpace=BCS_BoneSpace
    alphaBoolBlend {blendOption=Linear}
    alphaScaleBiasClamp {bInterpResult=true}
    boneToModify {boneName=dyn_cape_01_MSTR}
    rotation {pitch=10}
   }
   n13 = Get GravityOverride()
   n14 = Anim Dynamics - Bone: dyn_cape_01!(ComponentPose=@n12, Alpha=1, GravityOverride:Vector=@n13, ExternalForce="0,0,0") -> (Pose) {
    angularDampingOverride=1
    angularSpringConstant=10000
    bAngularSpring=true
    bOverrideAngularDamping=true
    bUseGravityOverride=true
    alphaBoolBlend {blendOption=Linear}
    boundBone {boneName=dyn_cape_01}
    componentAppliedLinearAccClamp=<3000 3000 3000>
    componentLinearAccScale=<0.4 0.4 0.4>
    componentLinearVelScale=<0.4 0.4 0.4>
    gravityOverride {z=-980}
    physicsBodyDefinitions [{
      boundBone {boneName=dyn_cape_01}
      boxExtents=<20 25 20>
      constraintSetup {angularLimitsMax=<15 5 15>, angularLimitsMin=<-15 -30 -15>, angularTarget=<1 0 0>}
      localJointOffset=<2 0 0>
     }]
   }
   n15 = Anim Dynamics - Bone: dyn_cape_02!(ComponentPose=@n14, Alpha=1, GravityOverride:Vector=@n13, ExternalForce="0,0,0") -> (Pose) {
    angularDampingOverride=1
    angularSpringConstant=10000
    bAngularSpring=true
    bOverrideAngularDamping=true
    bUseGravityOverride=true
    alphaBoolBlend {blendOption=Linear}
    boundBone {boneName=dyn_cape_02}
    componentAppliedLinearAccClamp=<3000 3000 3000>
    componentLinearAccScale=<0.4 0.4 0.4>
    componentLinearVelScale=<0.4 0.4 0.4>
    gravityOverride {z=-980}
    physicsBodyDefinitions [{
      boundBone {boneName=dyn_cape_02}
      boxExtents=<20 25 20>
      constraintSetup {angularLimitsMax=<15 5 15>, angularLimitsMin=<-15 -30 -15>, angularTarget=<1 0 0>}
      localJointOffset=<2 0 0>
     }]
   }
   n16 = Anim Dynamics - Bone: dyn_cape_03!(ComponentPose=@n15, Alpha=1, GravityOverride:Vector=@n13, ExternalForce="0,0,0") -> (Pose) {
    angularDampingOverride=1
    angularSpringConstant=10000
    bAngularSpring=true
    bOverrideAngularDamping=true
    bUseGravityOverride=true
    alphaBoolBlend {blendOption=Linear}
    boundBone {boneName=dyn_cape_03}
    componentAppliedLinearAccClamp=<3000 3000 3000>
    componentLinearAccScale=<0.4 0.4 0.4>
    componentLinearVelScale=<0.4 0.4 0.4>
    gravityOverride {z=-980}
    physicsBodyDefinitions [{
      boundBone {boneName=dyn_cape_03}
      boxExtents=<20 25 20>
      constraintSetup {angularLimitsMax=<15 5 15>, angularLimitsMin=<-15 -30 -15>, angularTarget=<1 0 0>}
      localJointOffset=<2 0 0>
     }]
   }
   n17 = Trail controller - Bone: dyn_cape_07_l!(ComponentPose=@n16, Alpha=1, RelaxationSpeedScale:float=1, FakeVelocity:Vector="0,0,0") -> (Pose) {
    bActorSpaceFakeVel=true
    bShowTrailLocation=true
    bUsePlanarLimit=true
    chainLength=7
    debugLifeTime=0.1
    maxDeltaTime=0.0333
    relaxationSpeedScale=1.2
    alphaScaleBiasClamp {bInterpResult=true, interpSpeedDecreasing=25, interpSpeedIncreasing=25}
    baseJoint {boneName=backpack_root}
    rotationLimits [
     {limitMax=<180 180 180>}
     {limitMax=<180 180 180>}
     {limitMax=<180 180 180>}
     {limitMax=<180 180 180>}
     {limitMax=<180 180 180>}
     {limitMax=<180 180 180>}
     {limitMax=<180 180 180>}
    ]
    rotationOffsets [<0 0 0>, <0 0 0>, <0 0 0>, <0 0 0>, <0 0 0>, <0 0 0>, <0 0 0>]
    trailBone {boneName=dyn_cape_07_l}
    trailRelaxationSpeed {editorCurveData {keys [<at 0 15 RCIM_Linear>,<at 0.5289 13.1838 RCIM_Linear>,<at 1 6 RCIM_Linear>]}}
   }
   n18 = Trail controller - Bone: dyn_cape_07_r!(ComponentPose=@n17, Alpha=1, RelaxationSpeedScale:float=1, FakeVelocity:Vector="0,0,0") -> (Pose) {
    bActorSpaceFakeVel=true
    bShowTrailLocation=true
    bUsePlanarLimit=true
    chainLength=7
    debugLifeTime=0.1
    maxDeltaTime=0.0333
    relaxationSpeedScale=1.2
    alphaScaleBiasClamp {bInterpResult=true, interpSpeedDecreasing=25, interpSpeedIncreasing=25}
    baseJoint {boneName=backpack_root}
    rotationLimits [
     {limitMax=<180 180 180>}
     {limitMax=<180 180 180>}
     {limitMax=<180 180 180>}
     {limitMax=<180 180 180>}
     {limitMax=<180 180 180>}
     {limitMax=<180 180 180>}
     {limitMax=<180 180 180>}
    ]
    rotationOffsets [<0 0 0>, <0 0 0>, <0 0 0>, <0 0 0>, <0 0 0>, <0 0 0>, <0 0 0>]
    trailBone {boneName=dyn_cape_07_r}
    trailRelaxationSpeed {editorCurveData {keys [<at 0 15 RCIM_Linear>,<at 0.5289 13.1838 RCIM_Linear>,<at 1 6 RCIM_Linear>]}}
   }
   Save cached pose 'trailControlCache'(Pose=Component To Local(ComponentPose=@n18)) {cacheName=trailControlCache}
  ]}
 EventGraph {nodes [
   Event BlueprintUpdateAnimation()
   > n1 = Sequence!(execute) -> (then_0, then_1, then_2:Exec)
   > n2 = "Gravity Override State Function (Parameter Asset) (Nativized)"(
    CharacterPartAnimInstance=Self-Reference()
    GameTimeInSeconds=GetGameTimeInSeconds()
    Parameters=&/M_Med_BriskReign_Pack_Parameters
   )
   > Set GravityOverride(GravityOverride=@n2)
   n3 = Sequence(execute=@n1.then_2)
   > n4 = Branch(Condition=Get Char F() -> (Char F:bool))
   > Set Alpha M(Alpha M:double=0) -> (Output_Get:double)
   > Set Alpha F(Alpha F:double=1) -> (Output_Get:double)
   Set Alpha M(execute=@n4.else, Alpha M=1)
   > Set Alpha F(Alpha F=0)
   n5 = Branch(execute=@n3.then_1, Condition=Get bIsInFrontEnd())
   > Set Alpha FE(Alpha FE:double=0.3) -> (Output_Get:double)
   Set Alpha FE(execute=@n5.else, Alpha FE=1)
   n6 = Branch(execute=@n3.then_2, Condition=Get bIsPlayingEmote())
   > Set Emote On(Emote On:double=0) -> (Output_Get:double)
   > Set Emote Off(Emote Off:double=1) -> (Output_Get:double)
   Set Emote On(execute=@n6.else, Emote On=1)
   > Set Emote Off(Emote Off=0)
   n7 = GetOwnerSkeletalMeshComponent()
   n8 = BreakRotator(InRot=BreakTransform(InTransform=GetDeltaTransformFromRefPose(self=@n7, BoneName=spine_05, BaseName=spine_01)).Rotation)
   n9 = MapRangeClamped(Value=@n8.Yaw, InRangeA=-10, InRangeB=-50, OutRangeB=40)
   Set Backpack Rotate(execute=@n1.then_1, Backpack Rotate:Rotator=MakeRotator(Pitch=@n9, Yaw=float / float(A=@n9, B=5))) -> (Output_Get:Rotator)
   > Set Backpack Translation(Backpack Translation:Vector=MakeVector(Y=0, Z=MapRangeClamped(Value=@n8.Yaw, InRangeA=-15, InRangeB=-50, OutRangeB=10))) -> (Output_Get:Vector)
   n10 = GetOwnerSkeletalMeshComponent()
   n11 = BreakRotator(InRot=BreakTransform(InTransform=GetDeltaTransformFromRefPose(self=@n10, BoneName=spine_05, BaseName=spine_01)).Rotation)
   n12 = MapRangeClamped(Value=@n11.Yaw, InRangeA=10, InRangeB=20, OutRangeB=5)
   n13 = float / float(A=@n12, B=5)
   > Set Backpack Rotate FE(Backpack Rotate FE:Rotator=MakeRotator(Roll=@n12, Pitch=@n13, Yaw=@n13)) -> (Output_Get:Rotator)
   n14 = MapRangeClamped(Value=@n11.Yaw, InRangeA=10, InRangeB=20, OutRangeB=-10)
   n15 = float / float(A=@n14, B=5)
   > Set Cape Rotate FE(Cape Rotate FE:Rotator=MakeRotator(Roll=@n15, Pitch=@n14, Yaw=@n15)) -> (Output_Get:Rotator)
   n16 = Event OnCharacterAndPickaxeCustomizationLoaded!() -> (
    OutputDelegate
    then
    CustomBaseMesh:Skeletal Mesh Component Object Reference
    CustomParts:Array of Custom Character Part Object References
    CustomMeshParts:Array of Skeletal Mesh Component Budgeted Object References
   ) {is=title K2Node_Event}
   n17 = Get Array Item with EFort Custom Part Type(Array:Array of Custom Character Part Object References=@n16.CustomParts) -> (Output:Custom Character Part Object Reference)
   > Set Char M(Char M:bool="Equal (Enum)"(
     A:EFortCustomGender Enum=Get GenderPermitted!(self:Custom Character Part Object Reference=@n17) -> (GenderPermitted:EFortCustomGender Enum=Invalid)
     B:EFortCustomGender Enum=Male
    )) -> (Output_Get:bool)
   > Set Char F(Char F:bool="Equal (Enum)"(A=Get GenderPermitted(self=@n17) -> (GenderPermitted=Invalid), B=Female)) -> (Output_Get:bool)
   TryGetPawnOwner()
  ]}
]
variables [
 Alpha F {type=number}
 Alpha FE {type=number}
 Alpha M {type=number}
 Backpack Rotate {type=Rotator}
 Backpack Rotate FE {type=Rotator}
 Backpack Translation {type=Vector}
 Cape Rotate FE {type=Rotator}
 Char F {type=boolean}
 Char M {type=boolean}
 Emote Off {type=number}
 Emote On {type=number}
 GravityOverride {type=Vector}
]
